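# Conversation database for the Season 37 Weapon Expertise vending machine, which sells loot and lets players permanently unlock new sales.
# Entry routes through five SaleChoice quest gates (active, not completed) into a shared unlock menu, or to a Sales node.
# Unlock options hide once their quest or token is owned; confirmations only carry per-sale choice tags, with no reward logic here.
# Sales offers five service-tag-gated SellLoot slots that loop through Repeat.
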ConversationDatabase /BattlePassS37_Quests/Items/Conversations/Conversations_VendingMachine_Expertise
entry_points [{tag=Conversation.Entry.VendingMachine.Expertise, destinations [#1]}]
nodes [
 ConversationEntryPointNode_2 {
  id=#1
  class=ConversationEntryPointNode
  outputs [
   #2
   #3
   #4
   #5
   #6
   #7
  ]
  properties {entryTag {tagName=Conversation.Entry.VendingMachine.Expertise}}
 }
 Speech_C_61 {
  id=#7
  class=Speech_C
  outputs [
   #8
   #9
   #10
   #11
   #12
  ]
  properties {bLookupViaServiceProviderTag=true, nodeName=Sales}
 }
 SellLoot_C_18 {
  id=#12
  class=SellLoot_C
  outputs [#13]
  properties {sellSlot=5}
  sub_nodes [
   SideEffect_PlayServiceUsedEffects_C_5 {class=SideEffect_PlayServiceUsedEffects_C, properties {bShowPropertyDetails=false}}
   FortConversationRequirement_HasService_6 {class=FortConversationRequirement_HasService, properties {serviceTag {tagName=Tandem.Service.Sell5}}}
  ]
 }
 Repeat_C_0 {id=#13, class=Repeat_C}
 SellLoot_C_16 {
  id=#11
  class=SellLoot_C
  outputs [#13]
  properties {sellSlot=4}
  sub_nodes [
   SideEffect_PlayServiceUsedEffects_C_23 {class=SideEffect_PlayServiceUsedEffects_C, properties {bShowPropertyDetails=false}}
   FortConversationRequirement_HasService_19 {class=FortConversationRequirement_HasService, properties {serviceTag {tagName=Tandem.Service.Sell4}}}
  ]
 }
 SellLoot_C_17 {
  id=#10
  class=SellLoot_C
  outputs [#13]
  properties {sellSlot=3}
  sub_nodes [
   SideEffect_PlayServiceUsedEffects_C_20 {class=SideEffect_PlayServiceUsedEffects_C, properties {bShowPropertyDetails=false}}
   FortConversationRequirement_HasService_16 {class=FortConversationRequirement_HasService, properties {serviceTag {tagName=Tandem.Service.Sell3}}}
  ]
 }
 SellLoot_C_15 {
  id=#9
  class=SellLoot_C
  outputs [#13]
  properties {sellSlot=2}
  sub_nodes [
   SideEffect_PlayServiceUsedEffects_C_21 {class=SideEffect_PlayServiceUsedEffects_C, properties {bShowPropertyDetails=false}}
   FortConversationRequirement_HasService_17 {class=FortConversationRequirement_HasService, properties {serviceTag {tagName=Tandem.Service.Sell2}}}
  ]
 }
 SellLoot_C_2 {
  id=#8
  class=SellLoot_C
  outputs [#13]
  sub_nodes [
   SideEffect_PlayServiceUsedEffects_C_22 {class=SideEffect_PlayServiceUsedEffects_C, properties {bShowPropertyDetails=false}}
   FortConversationRequirement_HasService_18 {class=FortConversationRequirement_HasService, properties {serviceTag {tagName=Tandem.Service.Sell1}}}
  ]
 }
 Speech_C_54 {
  id=#6
  class=Speech_C
  outputs [#14]
  properties {bLookupViaServiceProviderTag=true, generalConfig {defaultMessage="[Are you sure?]"}}
  sub_nodes [
   FortConversationRequirement_ControllerRequirement_65 {
    class=FortConversationRequirement_ControllerRequirement
    properties {
     nodeName=Require: Is Quest Active
     requirement {
      class=FortControllerRequirement_IsQuestActive
      properties {
       questDefinition=/BattlePassS37_Quests/Items/QuestItems/WeaponExpertise/OneTimeUnlock/Quest_S37_WeaponExpertise_SaleChoice_05
       questDefinitions [
        /BattlePassS37_Quests/Items/QuestItems/WeaponExpertise/OneTimeUnlock/Quest_S37_WeaponExpertise_SaleChoice_01
        /BattlePassS37_Quests/Items/QuestItems/WeaponExpertise/OneTimeUnlock/Quest_S37_WeaponExpertise_SaleChoice_02
        /BattlePassS37_Quests/Items/QuestItems/WeaponExpertise/OneTimeUnlock/Quest_S37_WeaponExpertise_SaleChoice_03
        /BattlePassS37_Quests/Items/QuestItems/WeaponExpertise/OneTimeUnlock/Quest_S37_WeaponExpertise_SaleChoice_04
        &Quest_S37_WeaponExpertise_SaleChoice_05
       ]
      }
     }
    }
   }
   FortConversationRequirement_ControllerRequirement_66 {
    class=FortConversationRequirement_ControllerRequirement
    properties {
     nodeName=Require: Has Completed Quest
     requirement {
      class=FortControllerRequirement_HasCompletedQuest
      properties {
       bInverseCondition=true
       questDefinition=&Quest_S37_WeaponExpertise_SaleChoice_05
       questDefinitions [
        &Quest_S37_WeaponExpertise_SaleChoice_01
        &Quest_S37_WeaponExpertise_SaleChoice_02
        &Quest_S37_WeaponExpertise_SaleChoice_03
        &Quest_S37_WeaponExpertise_SaleChoice_04
        &Quest_S37_WeaponExpertise_SaleChoice_05
       ]
      }
     }
    }
   }
  ]
 }
 Speech_C_52 {
  id=#14
  class=Speech_C
  outputs [
   #15
   #16
   #17
   #18
   #19
  ]
  properties {bLookupViaServiceProviderTag=true, generalConfig {defaultMessage="[Are you sure?]"}}
  sub_nodes [ConversationChoiceNode_75 {class=ConversationChoiceNode, properties {defaultChoiceDisplayText=Choose a new sale to permanently unlock}}]
 }
 Speech_C_50 {
  id=#19
  class=Speech_C
  outputs [#20, #21]
  properties {
   bLookupViaServiceProviderTag=true
   softTaskIcon=/QuestCategories/Items/QuestRewards/S37Expertise/T_UI_Boon_ExtendedMag
   generalConfig {defaultMessage="[Are you sure?]"}
  }
  sub_nodes [
   ConversationChoiceNode_1 {class=ConversationChoiceNode, properties {defaultChoiceDisplayText=Unlock Sale: Extended Magazine Boon}}
   FortConversationRequirement_ControllerRequirement_0 {
    class=FortConversationRequirement_ControllerRequirement
    properties {
     nodeName=Require: Has Completed Quest
     requirement {
      class=FortControllerRequirement_HasCompletedQuest
      properties {
       bInverseCondition=true
       questDefinition=/BattlePassS37_Quests/Items/QuestItems/WeaponExpertise/Hidden/Quest_S37_WeaponExpertise_UnlockSale_05
      }
     }
    }
   }
   FortConversationRequirement_ControllerRequirement_1 {
    class=FortConversationRequirement_ControllerRequirement
    properties {
     nodeName=Has token 5?
     requirement {
      class=UFortControllerRequirement_TokenThreshold_C
      properties {
       bInverseCondition=true
       itemTemplateId=Token:athena_s37_weaponexpertise_sale_unlock_05
       requiredQuantityThresholdConfig {requirement=GreaterThanOrEqual, quantity {value=1}}
      }
     }
    }
   }
  ]
 }
 Back_C_0 {
  id=#21
  class=Back_C
  sub_nodes [ConversationChoiceNode_4 {class=ConversationChoiceNode, properties {defaultChoiceDisplayText=Back}}]
 }
 Speech_C_1 {
  id=#20
  class=Speech_C
  outputs []
  properties {bLookupViaServiceProviderTag=true, bUseInlineConfig=false, softTaskIcon=/FortUILibrary/Textures/MaterialInstances/MI_UIKit_Icon_Check}
  sub_nodes [ConversationChoiceNode_2 {
    class=ConversationChoiceNode
    properties {
     defaultChoiceDisplayText=Unlock Extended Magazine Boon Sale?
     choiceTags {gameplayTags [{tagName=Conversation.Choice.Expertise.Unlock.Sale05}]}
    }
   }]
 }
 Speech_C_48 {
  id=#18
  class=Speech_C
  outputs [#22, #23]
  properties {
   bLookupViaServiceProviderTag=true
   softTaskIcon=/QuestCategories/Items/QuestRewards/S37Expertise/T_UI_Boon_AgileAiming
   generalConfig {defaultMessage="[Are you sure?]"}
  }
  sub_nodes [
   ConversationChoiceNode_33 {class=ConversationChoiceNode, properties {defaultChoiceDisplayText=Unlock Sale: Agile Aiming Boon}}
   FortConversationRequirement_ControllerRequirement_14 {
    class=FortConversationRequirement_ControllerRequirement
    properties {
     nodeName=Require: Has Completed Quest
     requirement {
      class=FortControllerRequirement_HasCompletedQuest
      properties {
       bInverseCondition=true
       questDefinition=/BattlePassS37_Quests/Items/QuestItems/WeaponExpertise/Hidden/Quest_S37_WeaponExpertise_UnlockSale_04
      }
     }
    }
   }
   FortConversationRequirement_ControllerRequirement_3 {
    class=FortConversationRequirement_ControllerRequirement
    properties {
     nodeName=Has token 4?
     requirement {
      class=UFortControllerRequirement_TokenThreshold_C
      properties {
       bInverseCondition=true
       itemTemplateId=Token:athena_s37_weaponexpertise_sale_unlock_04
       requiredQuantityThresholdConfig {requirement=GreaterThanOrEqual, quantity {value=1}}
      }
     }
    }
   }
  ]
 }
 Back_C_17 {
  id=#23
  class=Back_C
  sub_nodes [ConversationChoiceNode_31 {class=ConversationChoiceNode, properties {defaultChoiceDisplayText=Back}}]
 }
 Speech_C_49 {
  id=#22
  class=Speech_C
  outputs []
  properties {bLookupViaServiceProviderTag=true, bUseInlineConfig=false, softTaskIcon=&MI_UIKit_Icon_Check}
  sub_nodes [ConversationChoiceNode_32 {
    class=ConversationChoiceNode
    properties {defaultChoiceDisplayText=Unlock Agile Aiming Boon Sale?, choiceTags {gameplayTags [{tagName=Conversation.Choice.Expertise.Unlock.Sale04}]}}
   }]
 }
 Speech_C_46 {
  id=#17
  class=Speech_C
  outputs [#24, #25]
  properties {
   bLookupViaServiceProviderTag=true
   softTaskIcon=/QuestCategories/Items/QuestRewards/S37Expertise/T_UI_QuestReward_NanaimoBar
   generalConfig {defaultMessage="[Are you sure?]"}
  }
  sub_nodes [
   ConversationChoiceNode_72 {class=ConversationChoiceNode, properties {defaultChoiceDisplayText=Unlock Sale: Bug Blaster}}
   FortConversationRequirement_ControllerRequirement_42 {
    class=FortConversationRequirement_ControllerRequirement
    properties {
     nodeName=Require: Has Completed Quest
     requirement {
      class=FortControllerRequirement_HasCompletedQuest
      properties {
       bInverseCondition=true
       questDefinition=/BattlePassS37_Quests/Items/QuestItems/WeaponExpertise/Hidden/Quest_S37_WeaponExpertise_UnlockSale_03
      }
     }
    }
   }
   FortConversationRequirement_ControllerRequirement_6 {
    class=FortConversationRequirement_ControllerRequirement
    properties {
     nodeName=Has token 3?
     requirement {
      class=UFortControllerRequirement_TokenThreshold_C
      properties {
       bInverseCondition=true
       itemTemplateId=Token:athena_s37_weaponexpertise_sale_unlock_03
       requiredQuantityThresholdConfig {requirement=GreaterThanOrEqual, quantity {value=1}}
      }
     }
    }
   }
  ]
 }
 Back_C_16 {
  id=#25
  class=Back_C
  sub_nodes [ConversationChoiceNode_74 {class=ConversationChoiceNode, properties {defaultChoiceDisplayText=Back}}]
 }
 Speech_C_47 {
  id=#24
  class=Speech_C
  outputs []
  properties {bLookupViaServiceProviderTag=true, bUseInlineConfig=false, softTaskIcon=&MI_UIKit_Icon_Check}
  sub_nodes [ConversationChoiceNode_73 {
    class=ConversationChoiceNode
    properties {defaultChoiceDisplayText=Unlock Bug Blaster Sale?, choiceTags {gameplayTags [{tagName=Conversation.Choice.Expertise.Unlock.Sale03}]}}
   }]
 }
 Speech_C_44 {
  id=#16
  class=Speech_C
  outputs [#26, #27]
  properties {
   bLookupViaServiceProviderTag=true
   softTaskIcon=/QuestCategories/Items/QuestRewards/S37Expertise/T_UI_QuestReward_HeavyMiniGun_TidalCrane
   generalConfig {defaultMessage="[Are you sure?]"}
  }
  sub_nodes [
   ConversationChoiceNode_69 {class=ConversationChoiceNode, properties {defaultChoiceDisplayText=Unlock Sale: Leadspitter 3000}}
   FortConversationRequirement_ControllerRequirement_41 {
    class=FortConversationRequirement_ControllerRequirement
    properties {
     nodeName=Require: Has Completed Quest
     requirement {
      class=FortControllerRequirement_HasCompletedQuest
      properties {
       bInverseCondition=true
       questDefinition=/BattlePassS37_Quests/Items/QuestItems/WeaponExpertise/Hidden/Quest_S37_WeaponExpertise_UnlockSale_02
      }
     }
    }
   }
   FortConversationRequirement_ControllerRequirement_7 {
    class=FortConversationRequirement_ControllerRequirement
    properties {
     nodeName=Has token 2?
     requirement {
      class=UFortControllerRequirement_TokenThreshold_C
      properties {
       bInverseCondition=true
       itemTemplateId=Token:athena_s37_weaponexpertise_sale_unlock_02
       requiredQuantityThresholdConfig {requirement=GreaterThanOrEqual, quantity {value=1}}
      }
     }
    }
   }
  ]
 }
 Back_C_15 {
  id=#27
  class=Back_C
  sub_nodes [ConversationChoiceNode_71 {class=ConversationChoiceNode, properties {defaultChoiceDisplayText=Back}}]
 }
 Speech_C_45 {
  id=#26
  class=Speech_C
  outputs []
  properties {bLookupViaServiceProviderTag=true, bUseInlineConfig=false, softTaskIcon=&MI_UIKit_Icon_Check}
  sub_nodes [ConversationChoiceNode_70 {
    class=ConversationChoiceNode
    properties {defaultChoiceDisplayText=Unlock Leadspitter 3000 Sale?, choiceTags {gameplayTags [{tagName=Conversation.Choice.Expertise.Unlock.Sale02}]}}
   }]
 }
 Speech_C_2 {
  id=#15
  class=Speech_C
  outputs [#28, #29]
  properties {
   bLookupViaServiceProviderTag=true
   softTaskIcon=/QuestCategories/Items/QuestRewards/S37Expertise/T_UI_QuestReward_RocketLauncher_TidalCrane
   generalConfig {defaultMessage="[Are you sure?]"}
  }
  sub_nodes [
   ConversationChoiceNode_63 {class=ConversationChoiceNode, properties {defaultChoiceDisplayText=Unlock Sale: Swarmstrike}}
   FortConversationRequirement_ControllerRequirement_39 {
    class=FortConversationRequirement_ControllerRequirement
    properties {
     nodeName=Require: Has Completed Quest
     requirement {
      class=FortControllerRequirement_HasCompletedQuest
      properties {
       bInverseCondition=true
       questDefinition=/BattlePassS37_Quests/Items/QuestItems/WeaponExpertise/Hidden/Quest_S37_WeaponExpertise_UnlockSale_01
      }
     }
    }
   }
   FortConversationRequirement_ControllerRequirement_8 {
    class=FortConversationRequirement_ControllerRequirement
    properties {
     nodeName=Has token 1?
     requirement {
      class=UFortControllerRequirement_TokenThreshold_C
      properties {
       bInverseCondition=true
       itemTemplateId=Token:athena_s37_weaponexpertise_sale_unlock_01
       requiredQuantityThresholdConfig {requirement=GreaterThanOrEqual, quantity {value=1}}
      }
     }
    }
   }
  ]
 }
 Back_C_2 {
  id=#29
  class=Back_C
  sub_nodes [ConversationChoiceNode_65 {class=ConversationChoiceNode, properties {defaultChoiceDisplayText=Back}}]
 }
 Speech_C_3 {
  id=#28
  class=Speech_C
  outputs []
  properties {bLookupViaServiceProviderTag=true, bUseInlineConfig=false, softTaskIcon=&MI_UIKit_Icon_Check}
  sub_nodes [ConversationChoiceNode_64 {
    class=ConversationChoiceNode
    properties {defaultChoiceDisplayText=Unlock Swarmstrike Sale?, choiceTags {gameplayTags [{tagName=Conversation.Choice.Expertise.Unlock.Sale01}]}}
   }]
 }
 Speech_C_55 {
  id=#5
  class=Speech_C
  outputs [#14]
  properties {bLookupViaServiceProviderTag=true, generalConfig {defaultMessage="[Are you sure?]"}}
  sub_nodes [
   FortConversationRequirement_ControllerRequirement_47 {
    class=FortConversationRequirement_ControllerRequirement
    properties {
     nodeName=Require: Is Quest Active
     requirement {
      class=FortControllerRequirement_IsQuestActive
      properties {
       questDefinition=&Quest_S37_WeaponExpertise_SaleChoice_04
       questDefinitions [
        &Quest_S37_WeaponExpertise_SaleChoice_01
        &Quest_S37_WeaponExpertise_SaleChoice_02
        &Quest_S37_WeaponExpertise_SaleChoice_03
        &Quest_S37_WeaponExpertise_SaleChoice_04
        &Quest_S37_WeaponExpertise_SaleChoice_05
       ]
      }
     }
    }
   }
   FortConversationRequirement_ControllerRequirement_48 {
    class=FortConversationRequirement_ControllerRequirement
    properties {
     nodeName=Require: Has Completed Quest
     requirement {
      class=FortControllerRequirement_HasCompletedQuest
      properties {
       bInverseCondition=true
       questDefinition=&Quest_S37_WeaponExpertise_SaleChoice_04
       questDefinitions [
        &Quest_S37_WeaponExpertise_SaleChoice_01
        &Quest_S37_WeaponExpertise_SaleChoice_02
        &Quest_S37_WeaponExpertise_SaleChoice_03
        &Quest_S37_WeaponExpertise_SaleChoice_04
        &Quest_S37_WeaponExpertise_SaleChoice_05
       ]
      }
     }
    }
   }
  ]
 }
 Speech_C_60 {
  id=#4
  class=Speech_C
  outputs [#14]
  properties {bLookupViaServiceProviderTag=true, generalConfig {defaultMessage="[Are you sure?]"}}
  sub_nodes [
   FortConversationRequirement_ControllerRequirement_61 {
    class=FortConversationRequirement_ControllerRequirement
    properties {
     nodeName=Require: Is Quest Active
     requirement {
      class=FortControllerRequirement_IsQuestActive
      properties {
       questDefinition=&Quest_S37_WeaponExpertise_SaleChoice_03
       questDefinitions [
        &Quest_S37_WeaponExpertise_SaleChoice_01
        &Quest_S37_WeaponExpertise_SaleChoice_02
        &Quest_S37_WeaponExpertise_SaleChoice_03
        &Quest_S37_WeaponExpertise_SaleChoice_04
        &Quest_S37_WeaponExpertise_SaleChoice_05
       ]
      }
     }
    }
   }
   FortConversationRequirement_ControllerRequirement_62 {
    class=FortConversationRequirement_ControllerRequirement
    properties {
     nodeName=Require: Has Completed Quest
     requirement {
      class=FortControllerRequirement_HasCompletedQuest
      properties {
       bInverseCondition=true
       questDefinition=&Quest_S37_WeaponExpertise_SaleChoice_03
       questDefinitions [
        &Quest_S37_WeaponExpertise_SaleChoice_01
        &Quest_S37_WeaponExpertise_SaleChoice_02
        &Quest_S37_WeaponExpertise_SaleChoice_03
        &Quest_S37_WeaponExpertise_SaleChoice_04
        &Quest_S37_WeaponExpertise_SaleChoice_05
       ]
      }
     }
    }
   }
  ]
 }
 Speech_C_57 {
  id=#3
  class=Speech_C
  outputs [#14]
  properties {bLookupViaServiceProviderTag=true, generalConfig {defaultMessage="[Are you sure?]"}}
  sub_nodes [
   FortConversationRequirement_ControllerRequirement_51 {
    class=FortConversationRequirement_ControllerRequirement
    properties {
     nodeName=Require: Is Quest Active
     requirement {
      class=FortControllerRequirement_IsQuestActive
      properties {
       questDefinition=&Quest_S37_WeaponExpertise_SaleChoice_02
       questDefinitions [
        &Quest_S37_WeaponExpertise_SaleChoice_01
        &Quest_S37_WeaponExpertise_SaleChoice_02
        &Quest_S37_WeaponExpertise_SaleChoice_03
        &Quest_S37_WeaponExpertise_SaleChoice_04
        &Quest_S37_WeaponExpertise_SaleChoice_05
       ]
      }
     }
    }
   }
   FortConversationRequirement_ControllerRequirement_52 {
    class=FortConversationRequirement_ControllerRequirement
    properties {
     nodeName=Require: Has Completed Quest
     requirement {
      class=FortControllerRequirement_HasCompletedQuest
      properties {
       bInverseCondition=true
       questDefinition=&Quest_S37_WeaponExpertise_SaleChoice_02
       questDefinitions [
        &Quest_S37_WeaponExpertise_SaleChoice_01
        &Quest_S37_WeaponExpertise_SaleChoice_02
        &Quest_S37_WeaponExpertise_SaleChoice_03
        &Quest_S37_WeaponExpertise_SaleChoice_04
        &Quest_S37_WeaponExpertise_SaleChoice_05
       ]
      }
     }
    }
   }
  ]
 }
 Speech_C_53 {
  id=#2
  class=Speech_C
  outputs [#14]
  properties {bLookupViaServiceProviderTag=true, generalConfig {defaultMessage="[Are you sure?]"}}
  sub_nodes [
   FortConversationRequirement_ControllerRequirement_55 {
    class=FortConversationRequirement_ControllerRequirement
    properties {
     nodeName=Require: Is Quest Active
     requirement {
      class=FortControllerRequirement_IsQuestActive
      properties {
       questDefinition=&Quest_S37_WeaponExpertise_SaleChoice_01
       questDefinitions [
        &Quest_S37_WeaponExpertise_SaleChoice_01
        &Quest_S37_WeaponExpertise_SaleChoice_02
        &Quest_S37_WeaponExpertise_SaleChoice_03
        &Quest_S37_WeaponExpertise_SaleChoice_04
        &Quest_S37_WeaponExpertise_SaleChoice_05
       ]
      }
     }
    }
   }
   FortConversationRequirement_ControllerRequirement_56 {
    class=FortConversationRequirement_ControllerRequirement
    properties {
     nodeName=Require: Has Completed Quest
     requirement {
      class=FortControllerRequirement_HasCompletedQuest
      properties {
       bInverseCondition=true
       questDefinition=&Quest_S37_WeaponExpertise_SaleChoice_01
       questDefinitions [
        &Quest_S37_WeaponExpertise_SaleChoice_01
        &Quest_S37_WeaponExpertise_SaleChoice_02
        &Quest_S37_WeaponExpertise_SaleChoice_03
        &Quest_S37_WeaponExpertise_SaleChoice_04
        &Quest_S37_WeaponExpertise_SaleChoice_05
       ]
      }
     }
    }
   }
  ]
 }
]
